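# Niagara swing trail effect for the Punk Devil Summer Female pickaxe, drawn as a short-lived pink ribbon.
# A single SwingTrail emitter spawns ribbon particles at a steady rate with a color multiplied into high HDR values.
# Initial ribbon width grows with the emitter's normalized loop age, then a curve tapers each particle's width to zero over its brief lifetime.
# The system exposes skeletal mesh and ribbon ID offset user parameters.
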NiagaraSystem /SinisterMelon/Weapons/FORT_Melee/Pickaxe_Punk_Devil_Summer_Female/FX/NS_Pickaxe_Punk_Devil_Summer_Female_Swing
dependencies {usedDataInterfaces [&NiagaraDataInterfaceCurve], usedDynamicInputs [], usedModules [&SystemState]}
emitters {SwingTrail {
  emitterUpdateScript [
   &EmitterState()
   &FN_SpawnRibbon_v2(SpawnRate=75)
  ]
  particleSpawnScript [&FN_InitializeRibbon_v3(
    Color:LinearColor=&MultiplyLinearColor(
     Linear Color=<rgba 1 0.0267 0.1709 1>
     Scale Factor=&MakeVector4_FromLinearColor(LinearColor=<rgba 250 65.1043 209.3785 1>)
    )
    Lifetime=0.17
    Ribbon Width:NiagaraFloat=&Lerp_Float(A=0, B=26, Alpha=$Emitter.NormalizedLoopAge)
    Write Material Random=false
   )]
  particleUpdateScript [
   &SolveForcesAndVelocity(Clamp Velocity=false, Limit Acceleration=false)
   /Game/Effects/Niagara/Modules/Contrails/FN_UpdateRibbon/FN_UpdateRibbon_v3(Initial Color:LinearColor=$Particles.Initial.Color, Initial RibbonWidth:NiagaraFloat=$Particles.Initial.RibbonWidth)
   &ScaleRibbonWidth(Ribbon Width Scale=&FloatFromCurve(FloatCurve=DataInterface() {propertyValues {Curve {keys [<at 0 1.5>,<at 1 0>]}}}))
  ]
  rendererClasses [&NiagaraRibbonRendererProperties]
  renderers [{rendererClass=&NiagaraRibbonRendererProperties}]
 }}
summary {
 systemName=NS_Pickaxe_Punk_Devil_Summer_Female_Swing
 emitters [{emitterName=SwingTrail, rendererClasses [&NiagaraRibbonRendererProperties]}]
 userVariables [
  User.RibbonIDOffset {type=NiagaraInt32}
  User.SkeletalMesh {
   type=NiagaraDataInterfaceSkeletalMesh
   defaultValue {
    struct=NiagaraExt_VariableValue_DataInterface
    value {dataInterface=&self:NiagaraDataInterfaceSkeletalMesh_3}
   }
  }
 ]
}
